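# Niagara data-channel handler that draws sniper projectile visuals for Jethro weapons from a shared system-level reader.
# Each emitter spawns from channel data, tinting particles from an eight-color array indexed by VariantIndex.
# Active Mesh_Top (mesh plus sprite) and Trail_Mesh orient along projectile direction, stretching with previous-to-current position distance; the trail also shrinks over life.
# They discard start-position particles and cull non-end ones on update; ribbon Trail and Lights emitters are disabled.
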NiagaraSystem /JethroWeaponGameplay/Assets/FX/LW_Projectile_Sniper/NS_NDC_Jethro_ProjectilesHandler_LW
dependencies {
 usedDataInterfaces [
  &NiagaraDataInterfaceArrayFloat
  &NiagaraDataInterfaceCurve
  &NiagaraDataInterfaceArrayColor
  /Script/Niagara.NiagaraDataInterfaceDataChannelRead
 ]
 usedDynamicInputs []
 usedModules [
  &SystemState
  /Niagara/Modules/System/CompleteIfUnused
  &self:SystemSpawnScript.SystemScriptSource.SystemScriptGraph.NiagaraNodeAssignment_0.SetVariables_#1
 ]
}
emitters {
 Lights {
  bEnabled=false
  emitterUpdateScript [
   &EmitterState()
   &self:Spawn_NDC_SW_01(Reader:NiagaraDataInterfaceDataChannelRead=$System.NDCReader) {module=Spawn_NDC_SW_01}
  ]
  particleSpawnScript [
   &self:InitializeParticle_NDC_SW(
    End Lifetime=true
    End Lifetime Min=0.05
    Intro Duration=0.0167
    Intro Time=true
    LOD Output Scale Min=0.5
    Length by Velocity=true
    Lifetime=0.0167
    Scale=25
    Use LOD Distance Scaling=true
   ) {module=InitializeParticle_NDC_SW}
   &KillParticles(Kill Particles=$Particles.IsStartPosition)
   &InitializeParticle(
    Color=&SelectLinearColorFromArray(
     Array Sampling Mode=::NewEnumerator1
     Direct Array Index=$Particles.VariantIndex
     Color Selection Array=DataInterface() {propertyValues {ColorData [
        <rgba 1 0 0 1>
        <rgba 1 0.2 0 1>
        <rgba 1 0.5 0 1>
        <rgba 0.1269 1 0 1>
        <rgba 0 0.8333 1 1>
        <rgba 0 0.0833 1 5>
        <rgba 0.1623 0 1 1>
        <rgba 1 1 1 1>
       ]}}
     Random Seed=?
    )
    Lifetime=0.125
    Position=$Particles.CurrentPosition
    Position Mode=::NewEnumerator1
    Ribbon Width=250
    Ribbon Width Mode=::NewEnumerator1
    Sprite Size Mode=::NewEnumerator3
    Uniform Sprite Size=100
    UsePositionOffset=false
    Write Material Random=true
   )
   /SetVariables(Particles.IsHitScan:NiagaraBool=&Boolean_AndOperation(
     Bool1=$Particles.IsEndPosition
     Bool2=&SetBoolByFloatComparison(
      A=&VectorLength(
       Vector=&SubtractVectorFromPosition(Position=$Particles.PreviousPosition, Vector=Linked() {linkedVariable {name=Particles.OriginPosition, type=NiagaraPosition}})
       Safe Length Calculation=false
      )
      B=0
      Comparison Type=::NewEnumerator4
     )
    )) {isSetParameters=true}
   &DynamicMaterialParameters(
    Index 0 Param 1=&RemapRange(
     Input Value To Remap=&MakeFloatFromInt(Int=$Particles.VariantIndex)
     Input Min=0
     Input Max=7
     Output Min=0
     Output Max=1
    )
    Index 0 Param 2=1
    Module.Index0Float0Write=true
    Module.Index0Float1Write=false
    Module.Index0Float2Write=false
    Module.Index0Float3Write=false
   )
  ]
  particleUpdateScript [
   &ParticleState()
   &KillParticles(Kill Particles=&Boolean_NotOperation(Bool=$Particles.IsEndPosition)) {module=KillParticles001}
   /SetVariables(Particles.HitScanAge:NiagaraFloat=&MakeCustomFloatFromBool(
     Bool=$Particles.IsHitScan
     True Float=$Particles.NormalizedAge
     False Float=&ClampFloat(Float=&Add_Float(A=&Multiply_Float(A=$Particles.NormalizedAge, B=0.5), B=0.66))
    )) {isSetParameters=true}
   /SetVariables(Particles.Position=$Particles.Initial.Position) {isSetParameters=true}
   /SetVariables(Particles.Position=&MakeCustomVectorFromBool(
     Bool=$Particles.IsEndPosition
     True Vector=&Lerp_Vector(
      A=Linked() {linkedVariable {name=Particles.PreviousPosition, type=NiagaraPosition}}
      B=Linked() {linkedVariable {name=Particles.Position, type=NiagaraPosition}}
      Alpha=&FloatFromCurve(
       FloatCurve=DataInterface() {propertyValues {Curve {keys [{arriveTangent=2.1933, time=0, value=0}, <at 0.66 1>]}}}
       CurveIndex=$Particles.HitScanAge
      )
     )
     False Vector=Linked() {linkedVariable {name=Particles.Position, type=NiagaraPosition}}
    )) {isSetParameters=true}
   /SetVariables() {isSetParameters=true}
   /SetVariables(Particles.SpritePosition:NiagaraPosition=&AddVectorToPosition(Position=$Particles.Position, Vector=<0 0 0>)) {isSetParameters=true}
  ]
  rendererClasses [&NiagaraLightRendererProperties, &NiagaraComponentRendererProperties]
  renderers [
   {rendererClass=&NiagaraLightRendererProperties, rendererIndex=0}
   {rendererClass=&NiagaraComponentRendererProperties, rendererIndex=1}
  ]
 }
 Mesh_Top {
  emitterUpdateScript [
   &EmitterState()
   &self:Spawn_NDC_SW_01(Reader:NiagaraDataInterfaceDataChannelRead=$System.NDCReader) {module=Spawn_NDC_SW_01}
  ]
  particleSpawnScript [
   &self:InitializeParticle_NDC_SW(
    End Lifetime=true
    End Lifetime Max=0.0667
    Lifetime=0.125
    Player Position:NiagaraPosition=&MakeVectorFromLinearColorRGB(Color=$NPC.FortniteNPC.PlayerPosition)
    Projectile Test Length=3000
   ) {module=InitializeParticle_NDC_SW}
   &KillParticles(Kill Particles=$Particles.IsStartPosition)
   &InitializeParticle(
    Color=&SelectLinearColorFromArray(
     Array Sampling Mode=::NewEnumerator1
     Direct Array Index=$Particles.VariantIndex
     Color Selection Array=DataInterface() {propertyValues {ColorData [
        <rgba 1 0 0 1>
        <rgba 1 0.2 0 1>
        <rgba 1 0.5 0 1>
        <rgba 0.1269 1 0 1>
        <rgba 0 0.8333 1 1>
        <rgba 0 0.0833 1 1>
        <rgba 0.1623 0 1 1>
        <rgba 1 1 1 1>
       ]}}
     Random Seed=?
    )
    Lifetime=$Particles.Lifetime
    Mesh Scale=<3 1 1>
    Mesh Scale Mode=::NewEnumerator4
    Position=$Particles.CurrentPosition
    Position Mode=::NewEnumerator1
    Sprite Size Mode=::NewEnumerator3
    Uniform Sprite Size=120
    UsePositionOffset=false
    Write Material Random=true
   )
   /SetVariables(Particles.IsHitScan:NiagaraBool=&Boolean_AndOperation(
     Bool1=$Particles.IsEndPosition
     Bool2=&SetBoolByFloatComparison(
      A=&VectorLength(
       Vector=&SubtractVectorFromPosition(Position=$Particles.PreviousPosition, Vector=Linked() {linkedVariable {name=Particles.OriginPosition, type=NiagaraPosition}})
       Safe Length Calculation=false
      )
      B=0
      Comparison Type=::NewEnumerator4
     )
    )) {isSetParameters=true, module=SetVariables_#2}
   &InitialMeshOrientation(
    Mesh Orientation Mode=::NewEnumerator2
    Orientation Coordinate Space=::World
    Orientation Vector=$Particles.ProjectileDirection
    Use Rotation Vector=false
   )
  ]
  particleUpdateScript [
   &ParticleState()
   &KillParticles(Kill Particles=&Boolean_NotOperation(Bool=$Particles.IsEndPosition)) {module=KillParticles001}
   /SetVariables(Particles.HitScanAge:NiagaraFloat=&MakeCustomFloatFromBool(
     Bool=$Particles.IsHitScan
     True Float=$Particles.NormalizedAge
     False Float=&ClampFloat(Float=&Add_Float(A=&Multiply_Float(A=$Particles.NormalizedAge, B=0.5), B=0.66))
    )) {isSetParameters=true, module=SetVariables_#3}
   /SetVariables(Particles.Position=$Particles.Initial.Position) {isSetParameters=true, module=SetVariables_#4}
   /SetVariables(Particles.DistanceCheck:NiagaraFloat=&VectorLength(Vector=&Subtract_Position(A=$Particles.Position, B=$Particles.PreviousPosition))) {isSetParameters=true}
   &ScaleSpriteSize(
    Scale Sprite Size Mode=::NewEnumerator0
    Uniform Scale Factor=&ClampFloat(
     Float=&FN_LOD_Distance_Remap(Start Distance=0, End Distance=8000, Minimum Output=1, Maximum Output=1)
     Min=1
    )
   )
   &DynamicMaterialParameters(
    Index 0 Param 4=1
    Index 0 Vector 1:Vector3f=&SubtractVectorFromPosition(
     Position=$Particles.CurrentPosition
     Vector=&Multiply_VectorByFloat(
      Vector=$Particles.ProjectileDirection
      Float=&MakeCustomFloatFromBool(
       Bool=&SetBoolByFloatComparison(A=$Particles.DistanceCheck, B=75, Comparison Type=::NewEnumerator6)
       True Float=$Particles.DistanceCheck
       False Float=75
      )
     )
    )
    Module.Index0Float3Write=true
    Module.Index0VectorWrite=true
    Write Parameter Index 0=::NewEnumerator3
   )
   &ScaleMeshSize(Scale Factor=&MakeVector(
     X=&MakeCustomFloatFromBool(
      Bool=&SetBoolByFloatComparison(A=$Particles.DistanceCheck, B=597, Comparison Type=::NewEnumerator6)
      True Float=&Divide_Float(A=$Particles.DistanceCheck, B=199)
      False Float=3
     )
     Y=1
     Z=1
    ))
  ]
  rendererClasses [&NiagaraMeshRendererProperties, &NiagaraSpriteRendererProperties]
  renderers [
   {rendererClass=&NiagaraMeshRendererProperties, rendererIndex=0}
   {rendererClass=&NiagaraSpriteRendererProperties, rendererIndex=1}
  ]
 }
 Trail {
  bEnabled=false
  emitterUpdateScript [
   &EmitterState()
   &self:Spawn_NDC_SW_01(Reader:NiagaraDataInterfaceDataChannelRead=$System.NDCReader) {module=Spawn_NDC_SW_01}
  ]
  particleSpawnScript [
   &self:InitializeParticle_NDC_SW(FallbackOrigin:NiagaraPosition=$Engine.Owner.Position, LOD Output Scale Max=2, Scale=12, Use LOD Distance Scaling=true) {module=InitializeParticle_NDC_SW}
   &InitializeParticle(
    Color=&SelectLinearColorFromArray(
     Array Sampling Mode=::NewEnumerator1
     Direct Array Index=$Particles.VariantIndex
     Color Selection Array=DataInterface() {propertyValues {ColorData [
        <rgba 1 0 0 1>
        <rgba 1 0.2 0 1>
        <rgba 1 0.5 0 1>
        <rgba 0.1269 1 0 1>
        <rgba 0 0.8333 1 1>
        <rgba 0 0.0833 1 1>
        <rgba 0.1623 0 1 1>
        <rgba 1 1 1 1>
       ]}}
     Random Seed=?
    )
    Lifetime=0.125
    Position=$Particles.CurrentPosition
    Position Mode=::NewEnumerator1
    Ribbon Width=$Particles.ProjectileScale
    Ribbon Width Mode=::NewEnumerator1
    UsePositionOffset=false
    Write Material Random=true
   )
  ]
  particleUpdateScript [
   &ParticleState()
   &DynamicMaterialParameters(
    Index 0 Param 1=&RemapRange(
     Input Value To Remap=&MakeFloatFromInt(Int=$Particles.VariantIndex)
     Input Min=0
     Input Max=7
     Output Min=0
     Output Max=1
    )
    Index 0 Param 2=&SelectFloatFromArray(Direct Array Index=$Particles.VariantIndex, Float Selection Array=DataInterface() {propertyValues {FloatData [8,4,1,4,1,8,1,1]}})
    Module.Index0Float0Write=true
    Module.Index0Float1Write=true
    Module.Index0Float2Write=false
    Module.Index0Float3Write=false
   )
   &ScaleColor(
    Scale Alpha=&FloatFromCurve(FloatCurve=DataInterface() {propertyValues {Curve {keys [<at 0.85 1>,<at 1 0>]}}}, Scale Curve=0.5)
    ScaleA=true
    ScaleRGB=true
   )
  ]
  rendererClasses [&NiagaraRibbonRendererProperties]
  renderers [{rendererClass=&NiagaraRibbonRendererProperties}]
 }
 Trail_Mesh {
  emitterUpdateScript [
   &EmitterState()
   &self:Spawn_NDC_SW_01(Reader:NiagaraDataInterfaceDataChannelRead=$System.NDCReader) {module=Spawn_NDC_SW_01}
  ]
  particleSpawnScript [
   &self:InitializeParticle_NDC_SW(End Lifetime=true, End Lifetime Max=0.0667, Lifetime=0.125) {module=InitializeParticle_NDC_SW}
   &KillParticles(Kill Particles=$Particles.IsStartPosition)
   &InitializeParticle(
    Color=&SelectLinearColorFromArray(
     Array Sampling Mode=::NewEnumerator1
     Direct Array Index=$Particles.VariantIndex
     Color Selection Array=DataInterface() {propertyValues {ColorData [
        <rgba 1 0 0 1>
        <rgba 1 0.2 0 1>
        <rgba 1 0.5 0 1>
        <rgba 0.1269 1 0 1>
        <rgba 0 0.8333 1 1>
        <rgba 0 0.0833 1 1>
        <rgba 0.1623 0 1 1>
        <rgba 1 1 1 1>
       ]}}
     Random Seed=?
    )
    Lifetime=$Particles.Lifetime
    Mesh Scale=<1 0.65 0.65>
    Mesh Scale Mode=::NewEnumerator4
    Position=$Particles.CurrentPosition
    Position Mode=::NewEnumerator1
    UsePositionOffset=false
    Write Material Random=true
   )
   /SetVariables(Particles.IsHitScan:NiagaraBool=&Boolean_AndOperation(
     Bool1=$Particles.IsEndPosition
     Bool2=&SetBoolByFloatComparison(
      A=&VectorLength(
       Vector=&SubtractVectorFromPosition(Position=$Particles.PreviousPosition, Vector=Linked() {linkedVariable {name=Particles.OriginPosition, type=NiagaraPosition}})
       Safe Length Calculation=false
      )
      B=0
      Comparison Type=::NewEnumerator4
     )
    )) {isSetParameters=true, module=SetVariables_#2}
   &InitialMeshOrientation(
    Mesh Axis To Orient=<-1 0 0>
    Mesh Orientation Mode=::NewEnumerator2
    Orientation Coordinate Space=::World
    Orientation Vector=$Particles.ProjectileDirection
    Use Rotation Vector=false
   )
   &DynamicMaterialParameters(
    Index 0 Param 2=1
    Index 0 Param 3=1
    Index 0 Param 4=1
    Module.Index0Float0Write=true
    Module.Index0Float1Write=true
    Module.Index0Float2Write=true
    Module.Index0Float3Write=true
   )
  ]
  particleUpdateScript [
   &ParticleState()
   &ScaleMeshSize(Scale Factor=&MakeVector(
     X=&FloatFromCurve(FloatCurve=DataInterface() {propertyValues {Curve {keys [
          {arriveTangent=-1.9285, interpMode=RCIM_Cubic, leaveTangent=-1.9285, tangentMode=RCTM_User, time=0, value=1}
          {interpMode=RCIM_Cubic, leaveTangent=-1.1173, time=1, value=0}
         ]}}})
     Y=1
     Z=1
    ))
   &KillParticles(Kill Particles=&Boolean_NotOperation(Bool=$Particles.IsEndPosition)) {module=KillParticles001}
   /SetVariables(Particles.HitScanAge:NiagaraFloat=&MakeCustomFloatFromBool(
     Bool=$Particles.IsHitScan
     True Float=$Particles.NormalizedAge
     False Float=&ClampFloat(Float=&Add_Float(A=&Multiply_Float(A=$Particles.NormalizedAge, B=0.5), B=0.66))
    )) {isSetParameters=true, module=SetVariables_#3}
   /SetVariables(Particles.Position=$Particles.Initial.Position) {isSetParameters=true, module=SetVariables_#4}
   /SetVariables(Particles.DistanceCheck:NiagaraFloat=&VectorLength(Vector=&Subtract_Position(A=$Particles.Position, B=$Particles.PreviousPosition))) {isSetParameters=true}
   &ScaleMeshSize(Scale Factor=&MakeVector(
     X=&MakeCustomFloatFromBool(
      Bool=&SetBoolByFloatComparison(A=$Particles.DistanceCheck, B=300, Comparison Type=::NewEnumerator6)
      True Float=&Divide_Float(A=$Particles.DistanceCheck, B=100)
      False Float=3.5
     )
     Y=1
     Z=1
    )) {module=ScaleMeshSize001}
   /SetVariables(Particles.Position=&AddVectorToPosition(
     Position=&MakeCustomVectorFromBool(
      Bool=$Particles.IsEndPosition
      True Vector=&Lerp_Vector(
       A=Linked() {linkedVariable {name=Particles.PreviousPosition, type=NiagaraPosition}}
       B=Linked() {linkedVariable {name=Particles.Position, type=NiagaraPosition}}
       Alpha=&FloatFromCurve(
        FloatCurve=DataInterface() {propertyValues {Curve {keys [{arriveTangent=2.1933, time=0, value=0}, <at 0.66 1>]}}}
        CurveIndex=$Particles.HitScanAge
       )
      )
      False Vector=Linked() {linkedVariable {name=Particles.Position, type=NiagaraPosition}}
     )
     Vector=&Multiply_VectorByFloat(Vector=$Particles.ProjectileDirection, Float=-150)
    )) {enabled=false, isSetParameters=true}
  ]
  rendererClasses [&NiagaraMeshRendererProperties]
  renderers [{rendererClass=&NiagaraMeshRendererProperties, rendererIndex=0}]
 }
}
summary {
 systemName=NS_NDC_Jethro_ProjectilesHandler_LW
 emitters [
  {bEnabled=false, emitterName=Trail, rendererClasses [&NiagaraRibbonRendererProperties]}
  {
   bEnabled=false
   emitterName=Lights
   rendererClasses [&NiagaraLightRendererProperties, &NiagaraComponentRendererProperties]
  }
  {emitterName=Mesh_Top, rendererClasses [&NiagaraMeshRendererProperties, &NiagaraSpriteRendererProperties]}
  {emitterName=Trail_Mesh, rendererClasses [&NiagaraMeshRendererProperties]}
 ]
 userVariables []
}
